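# Physics asset for the Parallel Comic Male two-handed pickaxe, covering the handle and its dangling spider-silk bones.
# A long capsule covers the handle, and smaller capsules cover the root silk bone and its three branch bones.
# Constraints form a branching tree: the root silk links to the handle and each branch links to that root.
# All joints share moderate swing limits and an almost-zero twist limit, so the silk can sway with little twisting.
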PhysicsAsset /BRCosmetics/Weapons/FORT_Melee/Pickaxe_Parallel_Comic_Male/Meshes/Parallel_Comic_Male_Axe_Physics
bodies {
 Melee_TwoHanded_Handle [{shapeType=Capsule, length=82.6338, radius=2.1342, center=<0.2405 -1.3323 17.5728>, rotation=<rot 0.2124 -89.9999 0.442>}]
 dyn_SpiderSilk [{shapeType=Capsule, length=2.7731, radius=3.75, center=<0.3357 -0.5625 -0.0939>, rotation=<rot -11.2973 -89.9999 -91.7335>}]
 dyn_SpiderSilk_01 [{
   shapeType=Capsule
   length=6.2014
   radius=2.8852
   shapeName=dyn_SpiderSilk_01_capsule
   center=<-2.1127 -2.0315 -0.5608>
   rotation=<rot -1.7689 -60.0028 86.7581>
  }]
 dyn_SpiderSilk_02 [{
   shapeType=Capsule
   length=5.7646
   radius=1.8069
   shapeName=dyn_SpiderSilk_02_capsule
   center=<-2.3065 -2.1227 -0.1892>
   rotation=<rot -1.2296 -50.2167 81.6952>
  }]
 dyn_SpiderSilk_03 [{
   shapeType=Capsule
   length=4.2837
   radius=2.6838
   shapeName=dyn_SpiderSilk_03_capsule
   center=<-2.0411 -1.8514 0.2373>
   rotation=<rot 10.8489 -59.6247 88.6045>
  }]
}
constraints [
 {bone1Name=dyn_SpiderSilk, bone2Name=Melee_TwoHanded_Handle, swing1LimitDegrees=45, swing2LimitDegrees=45, twistLimitDegrees=1}
 {bone1Name=dyn_SpiderSilk_01, bone2Name=dyn_SpiderSilk, swing1LimitDegrees=45, swing2LimitDegrees=45, twistLimitDegrees=1}
 {bone1Name=dyn_SpiderSilk_02, bone2Name=dyn_SpiderSilk, swing1LimitDegrees=45, swing2LimitDegrees=45, twistLimitDegrees=1}
 {bone1Name=dyn_SpiderSilk_03, bone2Name=dyn_SpiderSilk, swing1LimitDegrees=45, swing2LimitDegrees=45, twistLimitDegrees=1}
]
