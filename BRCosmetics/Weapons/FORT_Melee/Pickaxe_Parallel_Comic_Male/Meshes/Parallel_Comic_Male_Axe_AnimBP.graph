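# Animation Blueprint for the Parallel Comic Male pickaxe that adds physics-driven secondary motion, including to its spider silk bones.
# The ref pose feeds a RigidBody sim relative to b_AR_Root, then a spring AnimDynamics on dyn_SpiderSilk_rototain_01.
# A weighted constraint partly drives dyn_SpiderSilk from that bone on X, blended with itself.
# Two other rototain AnimDynamics nodes are disconnected from the output, and the event graph has no logic.
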AnimBlueprint /BRCosmetics/Weapons/FORT_Melee/Pickaxe_Parallel_Comic_Male/Meshes/Parallel_Comic_Male_Axe_AnimBP
target_skeleton=&/Parallel_Comic_Male_Axe_Skeleton
graphs [
 AnimGraph {nodes [
   n1 = Anim Dynamics - Bone: dyn_SpiderSilk_rototain_base(Alpha=1, ExternalForce="0,0,0") {
    angularDampingOverride=0.999
    angularSpringConstant=800
    bAngularSpring=true
    bGravityOverrideInSimSpace=true
    bOverrideAngularDamping=true
    gravityScale=0.5
    alphaBoolBlend {blendOption=Linear}
    boundBone {boneName=dyn_SpiderSilk_rototain_base}
    componentAppliedLinearAccClamp=<10000 10000 10000>
    componentLinearAccScale=<0.1 0.1 0.1>
    componentLinearVelScale=<0.1 0.1 0.1>
    physicsBodyDefinitions [{
      sphereCollisionRadius=0
      boundBone {boneName=dyn_SpiderSilk_rototain_base}
      boxExtents=<3 10 5>
      constraintSetup {angularTargetAxis=AxisY, twistAxis=AxisY, angularLimitsMax=<0 2 7>, angularLimitsMin=<0 -2 -3>, angularTarget=<0 0 0>}
      localJointOffset=<-5 -0.05 0>
     }]
    relativeSpaceBone {boneName=b_AR_Root}
   }
   Anim Dynamics - Bone: dyn_SpiderSilk_rototain(ComponentPose=@n1, Alpha=1, ExternalForce="0,0,0") {
    angularDampingOverride=0.999
    angularSpringConstant=800
    bAngularSpring=true
    bGravityOverrideInSimSpace=true
    bOverrideAngularDamping=true
    gravityScale=0.5
    alphaBoolBlend {blendOption=Linear}
    boundBone {boneName=dyn_SpiderSilk_rototain}
    componentAppliedLinearAccClamp=<10000 10000 10000>
    componentLinearAccScale=<0.1 0.1 0.1>
    componentLinearVelScale=<0.1 0.1 0.1>
    physicsBodyDefinitions [{
      sphereCollisionRadius=0
      boundBone {boneName=dyn_SpiderSilk_rototain}
      boxExtents=<3 5 5>
      constraintSetup {angularTargetAxis=AxisY, twistAxis=AxisY, angularLimitsMax=<0 2 9>, angularLimitsMin=<0 -2 -6>, angularTarget=<0 0 0>}
      localJointOffset=<-3 -0.05 -0.05>
     }]
    relativeSpaceBone {boneName=b_AR_Root}
   }
   n2 = RigidBody(ComponentPose=Local To Component(LocalPose=Local Space Ref Pose())) {
    bClampLinearTranslationLimitToRefPose=true
    bOverrideWorldGravity=true
    overridePhysicsAsset=&/Parallel_Comic_Male_Axe_Physics
    simulationSpace=BaseBoneSpace
    alphaBoolBlend {blendOption=Linear}
    baseBoneRef {boneName=b_AR_Root}
    componentAppliedLinearAccClamp=<3000 3000 3000>
    componentLinearAccScale=<0.334 0.334 0.334>
    componentLinearVelScale=<0.334 0.334 0.334>
    overrideWorldGravity {z=-700}
   }
   n3 = Anim Dynamics - Bone: dyn_SpiderSilk_rototain_01(ComponentPose=@n2, Alpha=1, ExternalForce="0,0,0") {
    angularDampingOverride=0.999
    angularSpringConstant=800
    bAngularSpring=true
    bGravityOverrideInSimSpace=true
    bOverrideAngularDamping=true
    gravityScale=0.5
    alphaBoolBlend {blendOption=Linear}
    boundBone {boneName=dyn_SpiderSilk_rototain_01}
    componentAppliedLinearAccClamp=<2000 2000 2000>
    componentLinearAccScale=<0.1 0.1 0.1>
    componentLinearVelScale=<0.1 0.1 0.1>
    physicsBodyDefinitions [{
      sphereCollisionRadius=0
      boundBone {boneName=dyn_SpiderSilk_rototain_01}
      boxExtents=<3 10 15>
      constraintSetup {angularTargetAxis=AxisY, twistAxis=AxisY, angularLimitsMax=<10 5 10>, angularLimitsMin=<-10 -5 -10>, angularTarget=<0 0 0>}
      localJointOffset=<-5 -0.05 0>
     }]
    relativeSpaceBone {boneName=b_AR_Root}
   }
   n4 = "Constraint - dyn_SpiderSilk (PP)"!(ComponentPose=@n3, Alpha=1, ConstraintWeights_0=0.8, ConstraintWeights_1:float=0.2) -> (Pose) {
    alphaBoolBlend {blendOption=Linear}
    boneToModify {boneName=dyn_SpiderSilk}
    constraintSetup [{perAxis {bY=false, bZ=false}, targetBone {boneName=dyn_SpiderSilk_rototain_01}}, {targetBone {boneName=dyn_SpiderSilk}}]
    constraintWeights [1,1]
   }
   Output Pose(Result=Component To Local(ComponentPose=@n4)) {layerGroup=None}
  ]}
 EventGraph {nodes [
   TryGetPawnOwner()
   Event BlueprintUpdateAnimation()
  ]}
]
variables []
